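; Weapon Blueprint for the updated Core SMG that adds a repeating camera shake while the gun fires.
; When fire FX start, it runs OnFired at once and sets a looping timer at the inverse of the weapon's firing rate.
; OnFired triggers the shared FiringCameraShake from WeaponFunctionLibrary.
; When fire FX stop, it clears the stored timer handle; BeginPlay and Tick add no logic.
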(event EventBeginPlay
  (|CallParentFunction) ; This node is disabled and will not be called.\nDrag off pins to build functionality.
)

(event EventTick
  (params (DeltaSeconds :type Float%20%28single-precision%29)))

(event EventOnPlayWeaponFireFX
  (params (bPersistentFire :type Boolean) (bSecondaryFire :type Boolean))
  (|CallParentFunction bPersistentFire bSecondaryFire)
  (bind _returnvalue (Utilities|Time|SetTimerbyEvent (delegate Custom|OnFired) (/ 1.0 (FortWeapon|GetFiringRate)) true))
  (Variables|Default|SetLoopingFireHandle _returnvalue)
  (CallFunction|OnFired))

(event Custom|OnFired
  (Class|WeaponFunctionLibrary|FiringCameraShake self))

(event EventOnStopWeaponFireFX
  (|CallParentFunction)
  (Utilities|Time|ClearandInvalidateTimerbyHandle (Variables|Default|GetLoopingFireHandle)))

(fn ConstructionScript
  (|CallParentFunction))
